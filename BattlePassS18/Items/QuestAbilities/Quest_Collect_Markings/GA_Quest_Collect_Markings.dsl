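; Quest gameplay ability that spawns Kill And Collect pickups when a killed pawn carries the required pawn type tag.
; Each valid class in the collect lookup gets a random drop count between scalable-float limits, or class defaults if either is -1.
; The first item spawns at a random XY offset from the pawn's capsule, each later one offset from the previous item, chaining the scatter.
; The ability ends after spawning or failed validation.
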(event Ability|EventActivateAbilityFromEvent
  (params (EventData :type Gameplay%20Event%20Data%20Structure%20%28by%20ref%29))
  (bind (_eventtag _instigator _target) (Utilities|Struct|BreakGameplayEventData EventData))
  (|CallParentFunction EventData)
  (Development|PrintString "@@@ ActivateAbilityFromEvent [GA_Quest_Collect_Markings]: CALLED!" false (Variables|Debug|GetbDebugAbility))
  ;[0.448 0.1557 1] Determine Pawn kind
  (bind _asfort_player_pawn_athena (Utilities|Casting|CastToFortPlayerPawnAthena _target)
    (:then
      (Variables|Default|SetPawnKilled _asfort_player_pawn_athena)
      (Development|PrintString "@@@@@@ ActivateAbilityFromEvent [GA_Quest_Collect_Markings]: Found PlayerPawnAthena..." false (Variables|Debug|GetbDebugAbility)))
    (:CastFailed
      (bind _asfort_pawn (Utilities|Casting|CastToFortPawn _target)
        (:then
          (Variables|Default|SetPawnKilled _asfort_pawn)
          (Development|PrintString "@@@@@@ ActivateAbilityFromEvent [GA_Quest_Collect_Markings]: Found FortPawn..." false (Variables|Debug|GetbDebugAbility)))
        (:CastFailed
          (Development|PrintString "@@@@@@ ActivateAbilityFromEvent [GA_Quest_Collect_Markings]: FAILED to cast kill target to valid Pawn..." false (Variables|Debug|GetbDebugAbility))
          (Ability|EndAbility)))))
  (if (GameplayTags|HasTag (Variables|/Script/FortniteGame.FortPawn|GetGameplayTags (Variables|Default|GetPawnKilled)) (Variables|Default|GetPawnTypeTag))
    (CallFunction|SpawnPickupItem)
    (else
      (Development|PrintString (Utilities|String|Append "@@@@@@ ActivateAbilityFromEvent [GA_Quest_Collect_Markings]: FAILED to find required faction tag on target Pawn = " (Utilities|GetDisplayName (Variables|Default|GetPawnKilled))) false (Variables|Debug|GetbDebugAbility))
      (Ability|EndAbility))))

(event Custom|SpawnPickupItem
  ;[0.04144 1 0.9801] Spawn Collect Items for our target found Pawn.
  ;;[0 0 0 0.6] For each item that this pawn type can spawn, handle the spawning logic independently
  (for (_array_element _array_index) (Class|GetClassDefaults (Variables|Default|GetPawnCollectLookup))
    (if (Utilities|IsValidClass _array_element)
      (bind _output_get (Variables|Default|SetCurrentKillAndCollectClass _array_element))
      (bind _randomspawncount (CallFunction|GetRandomSpawnCount _output_get))
      (if (> _randomspawncount 0)
        ;;;[0 0 0 0.8] NOTE: Exponential increase in complexity for more than one type of item that needs to be spawned
        (for _index (range 1 (+ _randomspawncount 1))
          (if (== _index 1)
            (bind _randomtransform (CallFunction|MakeRandonXYTransform (Transformation|GetWorldLocation (Variables|/Script/Engine.Character|GetCapsuleComponent (Variables|Default|GetPawnKilled)))))
            (bind _returnvalue (Building|SpawnBuildingGameplayActor (Variables|Default|GetCurrentKillAndCollectClass) _randomtransform (Variables|Default|GetPawnKilled)))
            (bind _output_get_2 (Variables|Default|SetPreviousSpawnedCollectItem _returnvalue))
            (Development|PrintString (Utilities|String|Append "@@@@@@ ActivateAbilityFromEvent [GA_Quest_Collect_Markings]: Spawned " (Utilities|GetDisplayName _output_get_2) "...") false (Variables|Debug|GetbDebugAbility))
            (else
              (bind _randomtransform_2 (CallFunction|MakeRandonXYTransform (Transformation|GetActorLocation (Variables|Default|GetPreviousSpawnedCollectItem))))
              (bind _returnvalue_2 (Building|SpawnBuildingGameplayActor (Variables|Default|GetCurrentKillAndCollectClass) _randomtransform_2 (Variables|Default|GetPawnKilled)))
              (bind _output_get_3 (Variables|Default|SetPreviousSpawnedCollectItem _returnvalue_2))
              (Development|PrintString (Utilities|String|Append "@@@@@@ ActivateAbilityFromEvent [GA_Quest_Collect_Markings]: Spawned " (Utilities|GetDisplayName _output_get_3) "...") false (Variables|Debug|GetbDebugAbility))))))
      (else
        (Development|PrintString (Utilities|String|Append "@@@@@@ ActivateAbilityFromEvent [GA_Quest_Collect_Markings]: " (Utilities|GetClassDisplayName (Variables|Default|GetPawnCollectLookup)) "has invalid or empty Child Class referemce at index " (Utilities|String|ToString(Integer) _array_index) "...") false (Variables|Debug|GetbDebugAbility)))))
  (Ability|EndAbility))

(fn GetRandomSpawnCount
  (params (ChildClassRef :type BP%20Athena%20Kill%20And%20Collect%20Parent%20Class%20Reference))
  (bind _maxdrop (LocalVariables|GetMaxDrop))
  (bind _returnvalue (ScalableFloat|GetValueatLevel _maxdrop))
  (bind _returnvalue_2 (Math|Float|Floor _returnvalue))
  (bind _mindrop (LocalVariables|GetMinDrop))
  (bind _returnvalue_3 (ScalableFloat|GetValueatLevel _mindrop))
  (bind _returnvalue_4 (Math|Float|Floor _returnvalue_3))
  (bind (_sm_droppeditem _fortplayercontroller _meshrelativescale _ambientniagarasystem _ambientparticle _ambientfx_transform _ambientsound _pickupsound _ambientaudiofadeduration _ambientcollisionvolumeradius _mindrop_2 _maxdrop_2) (Class|GetClassDefaults ChildClassRef))
  ;[0.1367 0.8802 1] Takes min and max range for specified type from either the child class blueprints or AthenaQuestData table (if not -1) to determine count of the item to be spawned on this execution of the ability.
  (if (and (!= _returnvalue_4 -1) (!= _returnvalue_2 -1))
    (LocalVariables|SetMinDropLocal _returnvalue_4)
    (LocalVariables|SetMaxDropLocal _returnvalue_2)
    (else
      (LocalVariables|SetMinDropLocal _mindrop_2)
      (LocalVariables|SetMaxDropLocal _maxdrop_2)))
  (if (and (>= (LocalVariables|GetMinDropLocal) 0) (<= (LocalVariables|GetMinDropLocal) (LocalVariables|GetMaxDropLocal)))
    (if (> (LocalVariables|GetMaxDropLocal) 0)
      (return (Math|Random|RandomIntegerinRange (LocalVariables|GetMinDropLocal) (LocalVariables|GetMaxDropLocal)))
      (else
        (Development|PrintString "@@@@@@ ActivateAbilityFromEvent [GA_Quest_Collect_Markings]: Invalid max spawn count, or max spawn count is 0!  Item will not spawn..." false (Variables|Debug|GetbDebugAbility))))
    (else
      (Development|PrintString "@@@@@@ ActivateAbilityFromEvent [GA_Quest_Collect_Markings]: Invalid min spawn count!  Item will not spawn..." false (Variables|Debug|GetbDebugAbility))))
  (return 0))

(fn MakeRandonXYTransform
  (params (StartingLocation :type Vector))
  ;[1 0.4545 0.1489] Spawns the item(s) in a random x-y axis arrangement from the location given (first will be from where the pawn is killed, subsequent from where the previous item spawned).
  (bind _output_get (LocalVariables|SetAdjustedVectorLocal (Utilities|MakeRandomVector StartingLocation 3e+01 6e+01)))
  (return (Math|Transform|MakeTransform (Math|Vector|MakeVector (.x _output_get) (.y _output_get) (.z StartingLocation)))))
